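# Physics asset for the Mascot Militia Burger pickaxe mesh, a two-handed melee cosmetic.
# It defines capsule bodies on the handle bone and on a dyn_wire bone, whose capsule is offset and rotated.
# A single constraint joins dyn_wire to the handle with tight swing and twist limits, allowing only slight angular movement of the wire part.
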PhysicsAsset /BRCosmetics/Weapons/FORT_Melee/Pickaxe_MascotMilitiaBurger/Meshes/Mascot_Militia_Burger_Axe_Physics
bodies {
 Melee_TwoHanded_Handle [{shapeType=Capsule, length=10, radius=5}]
 dyn_wire [{shapeType=Capsule, length=12, radius=10, center=<4 0 0>, rotation=<rot 90 0 0>}]
}
constraints [{bone1Name=dyn_wire, bone2Name=Melee_TwoHanded_Handle, swing1LimitDegrees=3, swing2LimitDegrees=3, twistLimitDegrees=3}]
